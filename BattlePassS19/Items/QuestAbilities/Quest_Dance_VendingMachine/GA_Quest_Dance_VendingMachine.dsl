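; Gameplay ability for a Battle Pass quest that credits the player for dancing near a vending machine.
; On event activation, a repeating timer sphere-overlaps around the pawn for emote vending machines with the matching tag.
; Consecutive successful checks accumulate time; at five seconds it sends the quest stat event if the emote tag is present, then ends either way.
; The ability binds cancellation to the player going DBNO.
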(event Ability|EventActivateAbilityFromEvent
  (params (EventData :type Gameplay%20Event%20Data%20Structure%20%28by%20ref%29))
  ;[1 1 1] Run first check and start looping check
  (|CallParentFunction EventData)
  (bind _output_get (Variables|GAVars|Data|SetOwningPlayerPawn (CallFunction|GetAvatarasFortPlayerPawn)))
  (Variables|GAVars|Data|SetOwningPlayerQuestManager (Quests|GetQuestManager (Fortnite|GetFortPlayerControllerfromActor _output_get) "Invalid"))
  (CallFunction|CheckforNearbyVendingMachines false)
  (bind _returnvalue (Utilities|Time|SetTimerbyEvent (delegate Custom|LoopCheck) (Variables|Default|GetCheckInterval) true))
  (Variables|Default|SetTimer_Check _returnvalue)
  (EventDispatchers|BindEventtoOnPlayerStartDBNO (Variables|GAVars|Data|GetOwningPlayerPawn) (EventDispatchers|CreateEvent :!event K2_CancelAbility)))

(event Custom|LoopCheck
  ;[1 1 1] Run first check and start looping check
  (CallFunction|CheckforNearbyVendingMachines false))

(event Custom|CheckForNearbyVendingMachines
  (params (LastCheck :type Boolean))
  ;[1 1 1] Handle checking for nearby vending machines
  ;;[1 1 1] If not on the last check, save the time this loop triggered (used by last check)
  (if (not LastCheck)
    (Variables|Default|SetLastLoopTime (Fortnite|Sync|Time|GetServerWorldTimeSeconds)))
  ;;[1 1 1] See if we are standing near a vending machine
  (Variables|Default|SetCurrentCheckSucceeded false)
  (bind _outactors (Collision|SphereOverlapActors (Transformation|GetActorLocation (Variables|GAVars|Data|GetOwningPlayerPawn)) (* (Variables|Default|GetRadiusInMeters) 1e+02) (Utilities|Array|MakeArray "ObjectTypeQuery1") "/Game/Building/ActorBlueprints/Stations/BP_StationProp_VendingMachine_Emote_Base.BP_StationProp_VendingMachine_Emote_Base_C"))
  ;;[1 1 1] Check each hit actor until we find a vending machine
  (bind _array_element (Utilities|Array|ForEachLoopwithBreak _outactors)
    (:LoopBody
      (bind _bdidactorhavetags (Class|AthenaQuestFunctionLibrary|AthenaCheckActorHasTags _array_element "(GameplayTags=((TagName=\"Building.Type.ItemContainer.VendingMachine\")))" true))
      (if _bdidactorhavetags
        (Variables|Default|SetCurrentCheckSucceeded true)))
    (:Completed
      _))
  (if (and (Variables|Default|GetLastCheckSucceeded) (Variables|Default|GetCurrentCheckSucceeded))
    (Variables|Default|SetAccumulatedSuccessTime (+ (Variables|Default|GetAccumulatedSuccessTime) (select LastCheck (- (Fortnite|Sync|Time|GetServerWorldTimeSeconds) (Variables|Default|GetLastLoopTime)) (Variables|Default|GetCheckInterval))) :!pinned true) ; Add the successful time
    (else
      _))
  (Variables|Default|SetLastCheckSucceeded (Variables|Default|GetCurrentCheckSucceeded) :!pinned true) ; Save off whether this check was successful
  ;;[1 1 1] If we have danced for at least a little, then send quest progress
  (if (>= (Variables|Default|GetAccumulatedSuccessTime) 5.0)
    (if (GameplayTags|HasMatchingGameplayTag (Ability|GetFortAbilitySystemComponent (Variables|GAVars|Data|GetOwningPlayerPawn)) "(TagName=\"Gameplay.Action.Player.Emote\")")
      (Quest|SendComplexCustomStatEvent (Variables|GAVars|Data|GetOwningPlayerQuestManager) 0 (Variables|Default|GetQuestTag) "(GameplayTags=)" 1)
      (Utilities|Time|ClearandInvalidateTimerbyHandle (Variables|Default|GetTimer_Check))
      (EventDispatchers|UnbindEventfromOnPlayerStartDBNO (Variables|GAVars|Data|GetOwningPlayerPawn) (EventDispatchers|CreateEvent :!event K2_CancelAbility))
      (Ability|EndAbility)
      (else
        (Variables|Default|SetAccumulatedSuccessTime 0.0)
        (Ability|EndAbility)))))
